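# Temporary Default Lit material for the RustedTrack frog/croc cutout prop that builds an animated, color-dependent pulsing glow or tint from its textures.
# The diffuse is masked by the inverted, contrast-boosted green channel of the mask texture.
# Its posterized, red-green-swapped colors get a time-oscillating HueShift and desaturation, offsetting a time-driven sine that modulates the masked diffuse.
# A normal map is also sampled; material pin connections are absent.
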Material /Hera/Props/RustedTrack/RustedTrack_FrogCrocCutout_A/Textures/M_FrogCutout_TEMP
graph {
 properties {bAutomaticallySetUsageInEditor=false, usedShadingModels=Default Lit}
 nodes [
  Constant()
  n1 = TextureSample() {texture=&/T_RustedTrack_FrogCrocCutout_A_A_D}
  n2 = TextureSample() {samplerType=SAMPLERTYPE_Masks, texture=&/T_RustedTrack_FrogCrocCutout_A_A_S}
  n3 = Multiply(A=@n2.G)
  n4 = Saturate(None=MaterialFunctionCall(In=OneMinus(None=@n3), Contrast=ScalarParameter() {defaultValue=6.696, parameterName=Param_1}) {materialFunction=&CheapContrast}.Result)
  n5 = Time()
  n6 = Multiply(A=ScalarParameter() {defaultValue=0.2858, parameterName=Param_2}, B=Sine(None=Divide(A=@n5) {constB=4}))
  n7 = Multiply(A=@n1.RGB) {constB=0.5}
  n8 = ScalarParameter() {defaultValue=0.2946, parameterName=Param_4}
  n9 = Saturate(None=Round(None=Add(A=@n7, B=@n8)))
  n10 = Constant3Vector() {constant=<rgba 0 1 0 0>}
  n11 = Constant3Vector() {constant=<rgba 1 0 0 0>}
  n12 = AppendVector(A=Multiply(A=DotProduct(A=@n9, B=@n10)), B=Multiply(A=DotProduct(A=@n9, B=@n11)))
  n13 = DotProduct(A=@n9, B=Constant3Vector() {constant=<rgba 0 0 1 0>})
  n14 = Round(None=Saturate(None=AppendVector(A=@n12, B=@n13)))
  n15 = Add(
   A=Time()
   B=Desaturation(None=MaterialFunctionCall(Hue Shift Percentage=@n6, Texture=@n14) {materialFunction=&HueShift}.Result)
  )
  Multiply(A=Multiply(A=ScalarParameter() {defaultValue=1}, B=Multiply(A=@n1.RGB, B=@n4)), B=Saturate(None=Sine(None=@n15)))
  TextureSample() {samplerType=SAMPLERTYPE_Normal, texture=&/T_RustedTrack_FrogCrocCutout_A_A_N}
 ]
}
